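; Deimos Brute Gold Reskin is a child pawn Blueprint that adds handling for an apparition skin effect and beam visual to its parent character.
; The construction script runs the parent's, then feeds FX_Color and the mesh's relative X scale into the NS_Deimos_Apparition_SkinFX component.
; On death, a custom event zeroes the mesh materials' BeamOpacity and deactivates that effect where cosmetic events can run, before the parent death logic executes.
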(event EventCustomOnDied
  (params (DamagedActor :type Actor%20Object%20Reference) (Damage :type Float%20%28single-precision%29) (InstigatedBy :type Controller%20Object%20Reference) (DamageCauser :type Actor%20Object%20Reference) (HitLocation :type Vector :default "0, 0, 0") (FHitComponent :type Primitive%20Component%20Object%20Reference) (BoneName :type Name :default "None") (Momentum :type Vector :default "0, 0, 0"))
  (CallFunction|CEHideBeamFXonDeimos)
  (|CallParentFunction DamagedActor Damage InstigatedBy DamageCauser HitLocation FHitComponent BoneName Momentum))

(event Custom|CE_HideBeamFXonDeimos
  (Networking|CanExecuteCosmeticEvents
    (:True
      (Rendering|Material|SetScalarParameterValueonMaterials (Variables|Character|GetMesh) "BeamOpacity" 0.0)
      (Components|Activation|Deactivate (Variables|Default|GetNS_Deimos_Apparition_SkinFX)))))

(fn ConstructionScript
  (bind _ns_deimos_apparition_skinfx (Variables|Default|GetNS_Deimos_Apparition_SkinFX))
  (|CallParentFunction)
  (Effects|Components|ParticleSystem|SetColorParameter _ns_deimos_apparition_skinfx "FX_Color" (Variables|Default|GetFX_Color))
  (Effects|Components|ParticleSystem|SetFloatParameter _ns_deimos_apparition_skinfx "FX_Scalar" (.x (Variables|/Script/Engine.SceneComponent|GetRelativeScale3D (Variables|Character|GetMesh)))))
